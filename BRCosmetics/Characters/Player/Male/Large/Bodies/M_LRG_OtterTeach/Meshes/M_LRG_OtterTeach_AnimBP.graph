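# Post-process style AnimBP for the OtterTeach large male body cosmetic, adding secondary motion to its shorts and string.
# It caches the attached parent's pose, then runs two Control Rig and AnimDynamics chains into a bool blend with an unwired selector.
# One chain adds corrective shorts offsets and arm-hidden bone tweaks.
# The event graph updates gravity and a state-based dynamics weight, neither read by the AnimGraph.
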AnimBlueprint /BRCosmetics/Characters/Player/Male/Large/Bodies/M_LRG_OtterTeach/Meshes/M_LRG_OtterTeach_AnimBP
target_skeleton=&/M_LRG_OtterTeach_Skeleton
graphs [
 AnimGraph {nodes [
   Anim Dynamics - Bone: dyn_shorts_l_03(Alpha=0.3, ExternalForce="0,0,0") {
    angularDampingOverride=1
    angularSpringConstant=20000
    bAngularSpring=true
    bOverrideAngularDamping=true
    bUseGravityOverride=true
    gravityScale=0.35
    simulationSpace=RootRelative
    alphaScaleBiasClamp {bInterpResult=true, interpSpeedDecreasing=5, interpSpeedIncreasing=15}
    boundBone {boneName=dyn_shorts_l_03}
    componentLinearAccScale=<1 1 1>
    componentLinearVelScale=<1 1 1>
    gravityOverride {z=-980}
    physicsBodyDefinitions [{
      boundBone {boneName=dyn_shorts_l_03}
      boxExtents=<20 30 30>
      constraintSetup {angularLimitsMax=<10 10 10>, angularLimitsMin=<-10 -10 -10>, angularTarget=<1 0 0>}
      localJointOffset=<1 0 0>
     }]
   }
   Anim Dynamics - Bone: dyn_shorts_l_03(Alpha=0.3, ExternalForce="0,0,0") {
    angularDampingOverride=1
    angularSpringConstant=20000
    bAngularSpring=true
    bOverrideAngularDamping=true
    bUseGravityOverride=true
    gravityScale=0.35
    simulationSpace=RootRelative
    alphaScaleBiasClamp {bInterpResult=true, interpSpeedDecreasing=5, interpSpeedIncreasing=15}
    boundBone {boneName=dyn_shorts_l_03}
    componentLinearAccScale=<1 1 1>
    componentLinearVelScale=<1 1 1>
    gravityOverride {z=-980}
    physicsBodyDefinitions [{
      boundBone {boneName=dyn_shorts_l_03}
      boxExtents=<20 30 30>
      constraintSetup {angularLimitsMax=<10 10 10>, angularLimitsMin=<-10 -10 -10>, angularTarget=<1 0 0>}
      localJointOffset=<1 0 0>
     }]
    simSpaceSettings {simSpaceAngularAlpha=1}
   }
   n1 = "Transform (Modify) Bone - Bone: base_shorts_l_01"!(ComponentPose, bAlphaBoolEnabled:bool=True, Translation="-4,0,0", Rotation="0,0,0", Scale="1,1,1") -> (Pose) {
    alphaInputType=Bool
    rotationMode=BMM_Additive
    rotationSpace=BCS_BoneSpace
    translationMode=BMM_Additive
    translationSpace=BCS_BoneSpace
    alphaBoolBlend {blendInTime=0.2, blendOutTime=0.1}
    boneToModify {boneName=base_shorts_l_01}
    translation {x=-4}
   }
   "Transform (Modify) Bone - Bone: base_shorts_l_02"!(ComponentPose=@n1, bAlphaBoolEnabled:bool=True, Translation="-5,0,0", Rotation="0,0,0", Scale="1,1,1") -> (Pose) {
    alphaInputType=Bool
    rotationMode=BMM_Additive
    rotationSpace=BCS_BoneSpace
    translationMode=BMM_Additive
    translationSpace=BCS_BoneSpace
    alphaBoolBlend {blendInTime=0.2, blendOutTime=0.1}
    boneToModify {boneName=base_shorts_l_02}
    translation {x=-5}
   }
   n2 = "Transform (Modify) Bone - Bone: base_shorts_l_04"!(ComponentPose, bAlphaBoolEnabled:bool=True, Translation="0,0,0", Rotation="0,0,0", Scale="1,1,1") -> (Pose) {
    alphaInputType=Bool
    rotationMode=BMM_Additive
    rotationSpace=BCS_BoneSpace
    translationMode=BMM_Additive
    translationSpace=BCS_BoneSpace
    alphaBoolBlend {blendInTime=0.2, blendOutTime=0.1}
    boneToModify {boneName=base_shorts_l_04}
   }
   n3 = "Transform (Modify) Bone - Bone: base_shorts_r_01"!(ComponentPose=@n2, bAlphaBoolEnabled:bool=True, Translation="1,-2.5,0", Rotation="0,0,0", Scale="1,1,1") -> (Pose) {
    alphaInputType=Bool
    rotationMode=BMM_Additive
    rotationSpace=BCS_BoneSpace
    translationMode=BMM_Additive
    translationSpace=BCS_BoneSpace
    alphaBoolBlend {blendInTime=0.2, blendOutTime=0.1}
    boneToModify {boneName=base_shorts_r_01}
    translation=<1 -2.5>
   }
   "Transform (Modify) Bone - Bone: base_shorts_r_02"!(ComponentPose=@n3, bAlphaBoolEnabled:bool=True, Translation="5,0,0", Rotation="0,10,0", Scale="1,1,1") -> (Pose) {
    alphaInputType=Bool
    rotationMode=BMM_Additive
    rotationSpace=BCS_BoneSpace
    translationMode=BMM_Additive
    translationSpace=BCS_BoneSpace
    alphaBoolBlend {blendInTime=0.2, blendOutTime=0.1}
    boneToModify {boneName=base_shorts_r_02}
    rotation {yaw=10}
    translation {x=5}
   }
   n4 = "Transform (Modify) Bone - Bone: base_shorts_l_03"!(ComponentPose, bAlphaBoolEnabled:bool=True, Translation="4.3,0.2,2.5", Rotation="-38.5,-2.8,2.6", Scale="1,1.5,1") -> (Pose) {
    alphaInputType=Bool
    rotationMode=BMM_Additive
    rotationSpace=BCS_BoneSpace
    scaleMode=BMM_Additive
    translationMode=BMM_Additive
    translationSpace=BCS_BoneSpace
    alphaBoolBlend {blendInTime=0.5, blendOutTime=0.1}
    boneToModify {boneName=base_shorts_l_03}
    rotation=<rot -38.5 -2.8 2.6>
    scale {y=1.5}
    translation=<4.3 0.2 2.5>
   }
   n5 = "Transform (Modify) Bone - Bone: base_shorts_r_03"!(ComponentPose=@n4, bAlphaBoolEnabled:bool=True, Translation="-7.788,-2.994,-0.877", Rotation="-30.666,-5.681,16.590", Scale="1,9,1") -> (Pose) {
    alphaInputType=Bool
    rotationMode=BMM_Additive
    rotationSpace=BCS_BoneSpace
    scaleMode=BMM_Additive
    translationMode=BMM_Additive
    translationSpace=BCS_BoneSpace
    alphaBoolBlend {blendInTime=0.5, blendOutTime=0.1}
    boneToModify {boneName=base_shorts_r_03}
    rotation=<rot -30.666 -5.681 16.59>
    scale {y=9}
    translation=<-7.788 -2.994 -0.877>
   }
   "Transform (Modify) Bone - Bone: base_shorts_r_04"!(ComponentPose=@n5, bAlphaBoolEnabled:bool=True, Translation="5.020,-8.000,-2.796", Rotation="0.009,0.005,-5.128", Scale="1,2,1") -> (Pose) {
    alphaInputType=Bool
    rotationMode=BMM_Additive
    rotationSpace=BCS_BoneSpace
    scaleMode=BMM_Additive
    translationMode=BMM_Additive
    translationSpace=BCS_BoneSpace
    alphaBoolBlend {blendInTime=0.2, blendOutTime=0.1}
    boneToModify {boneName=base_shorts_r_04}
    rotation=<rot 0.009 0.005 -5.128>
    scale {y=2}
    translation=<5.02 -8 -2.796>
   }
   n6 = Control Rig(Source=Use cached pose 'Input_pose'()) {controlRigAssetReference {controlRigAsset=&/M_LRG_OtterTeach_CtrlRig}}
   n7 = "Transform (Modify) Bone - Bone: base_shorts_r_04"(ComponentPose=Local To Component(LocalPose=@n6), bAlphaBoolEnabled=True, Translation="0,0,-1", Rotation="0,0,0", Scale="1,1,1") {
    alphaInputType=Bool
    rotationMode=BMM_Additive
    rotationSpace=BCS_BoneSpace
    translationMode=BMM_Additive
    translationSpace=BCS_BoneSpace
    alphaBoolBlend {blendInTime=0.2, blendOutTime=0.1}
    boneToModify {boneName=base_shorts_r_04}
    translation {z=-1}
   }
   n8 = "Transform (Modify) Bone - Bone: base_shorts_r_03"(ComponentPose=@n7, bAlphaBoolEnabled=True, Translation="5,0,-1", Rotation="20,0,0", Scale="1,1,1") {
    alphaInputType=Bool
    rotationMode=BMM_Additive
    rotationSpace=BCS_BoneSpace
    translationMode=BMM_Additive
    translationSpace=BCS_BoneSpace
    alphaBoolBlend {blendInTime=0.2, blendOutTime=0.1}
    boneToModify {boneName=base_shorts_r_03}
    rotation {pitch=20}
    translation {x=5, z=-1}
   }
   n9 = "Transform (Modify) Bone - Bone: base_shorts_r_02"(ComponentPose=@n8, bAlphaBoolEnabled=True, Translation="1,5,-2", Rotation="0,0,0", Scale="1,1,1") {
    alphaInputType=Bool
    rotationMode=BMM_Additive
    rotationSpace=BCS_BoneSpace
    translationMode=BMM_Additive
    translationSpace=BCS_BoneSpace
    alphaBoolBlend {blendInTime=0.2, blendOutTime=0.1}
    boneToModify {boneName=base_shorts_r_02}
    translation=<1 5 -2>
   }
   n10 = "Transform (Modify) Bone - Bone: base_shorts_l_03"(ComponentPose=@n9, bAlphaBoolEnabled=True, Translation="0,0,5", Rotation="25,0,0", Scale="1,1,1") {
    alphaInputType=Bool
    rotationMode=BMM_Additive
    rotationSpace=BCS_BoneSpace
    translationMode=BMM_Additive
    translationSpace=BCS_BoneSpace
    alphaBoolBlend {blendInTime=0.2, blendOutTime=0.1}
    boneToModify {boneName=base_shorts_l_03}
    rotation {pitch=25}
    translation {z=5}
   }
   n11 = "Transform (Modify) Bone - Bone: base_shorts_l_02"(ComponentPose=@n10, bAlphaBoolEnabled=True, Translation="0,0,0", Rotation="5,0,0", Scale="1,1,1") {
    alphaInputType=Bool
    rotationMode=BMM_Additive
    rotationSpace=BCS_BoneSpace
    translationMode=BMM_Additive
    translationSpace=BCS_BoneSpace
    alphaBoolBlend {blendInTime=0.2, blendOutTime=0.1}
    boneToModify {boneName=base_shorts_l_02}
    rotation {pitch=5}
   }
   n12 = "Transform (Modify) Bone - Bone: base_shorts_l_01"(ComponentPose=@n11, bAlphaBoolEnabled=True, Translation="0,0,1", Rotation="0,0,0", Scale="1,1,1") {
    alphaInputType=Bool
    rotationMode=BMM_Additive
    rotationSpace=BCS_BoneSpace
    translationMode=BMM_Additive
    translationSpace=BCS_BoneSpace
    alphaBoolBlend {blendInTime=0.2, blendOutTime=0.1}
    boneToModify {boneName=base_shorts_l_01}
    translation {z=1}
   }
   n13 = "Transform (Modify) Bone - Bone: base_shorts_r_04"(ComponentPose=@n12, bAlphaBoolEnabled=True, Translation="0,0,-2", Rotation="0,0,0", Scale="1,1,1") {
    alphaInputType=Bool
    rotationMode=BMM_Additive
    rotationSpace=BCS_BoneSpace
    translationMode=BMM_Additive
    translationSpace=BCS_BoneSpace
    alphaBoolBlend {blendInTime=0.2, blendOutTime=0.1}
    boneToModify {boneName=base_shorts_r_04}
    translation {z=-2}
   }
   n14 = "Transform (Modify) Bone - Bone: base_shorts_r_02"(ComponentPose=@n13, bAlphaBoolEnabled=True, Translation="0,0,-2", Rotation="0,0,0", Scale="1,1,1") {
    alphaInputType=Bool
    rotationMode=BMM_Additive
    rotationSpace=BCS_BoneSpace
    translationMode=BMM_Additive
    translationSpace=BCS_BoneSpace
    alphaBoolBlend {blendInTime=0.2, blendOutTime=0.1}
    boneToModify {boneName=base_shorts_r_02}
    translation {z=-2}
   }
   n15 = "Transform (Modify) Bone - Bone: base_shorts_r_03"(ComponentPose=@n14, bAlphaBoolEnabled=True, Translation="0,0,-0.5", Rotation="5,0,0", Scale="1,1,1") {
    alphaInputType=Bool
    rotationMode=BMM_Additive
    rotationSpace=BCS_BoneSpace
    translationMode=BMM_Additive
    translationSpace=BCS_BoneSpace
    alphaBoolBlend {blendInTime=0.2, blendOutTime=0.1}
    boneToModify {boneName=base_shorts_r_03}
    rotation {pitch=5}
    translation {z=-0.5}
   }
   n16 = "Transform (Modify) Bone - Bone: base_shorts_l_03"(ComponentPose=@n15, bAlphaBoolEnabled=True, Translation="0,0,0.5", Rotation="-5,0,0", Scale="1,1,1") {
    alphaInputType=Bool
    rotationMode=BMM_Additive
    rotationSpace=BCS_BoneSpace
    translationMode=BMM_Additive
    translationSpace=BCS_BoneSpace
    alphaBoolBlend {blendInTime=0.2, blendOutTime=0.1}
    boneToModify {boneName=base_shorts_l_03}
    rotation {pitch=-5}
    translation {z=0.5}
   }
   n17 = Anim Dynamics - Bone: dyn_shorts_l_01(ComponentPose=@n16, Alpha=1, ExternalForce="0,0,0") {
    angularDampingOverride=1
    angularSpringConstant=20000
    bAngularSpring=true
    bOverrideAngularDamping=true
    bUseGravityOverride=true
    gravityScale=0.35
    simulationSpace=RootRelative
    alphaScaleBiasClamp {bInterpResult=true, interpSpeedDecreasing=5, interpSpeedIncreasing=15}
    boundBone {boneName=dyn_shorts_l_01}
    componentLinearAccScale=<1 1 1>
    componentLinearVelScale=<1 1 1>
    gravityOverride {z=-980}
    physicsBodyDefinitions [{
      boundBone {boneName=dyn_shorts_l_01}
      boxExtents=<20 30 30>
      constraintSetup {angularLimitsMax=<0 0 5>, angularLimitsMin=<0 0 -5>, angularTarget=<1 0 0>}
      localJointOffset=<1 0 0>
     }]
   }
   n18 = Anim Dynamics - Bone: dyn_shorts_l_04(ComponentPose=@n17, Alpha=1, ExternalForce="0,0,0") {
    angularDampingOverride=1
    angularSpringConstant=20000
    bAngularSpring=true
    bOverrideAngularDamping=true
    bUseGravityOverride=true
    gravityScale=0.35
    simulationSpace=RootRelative
    alphaScaleBiasClamp {bInterpResult=true, interpSpeedDecreasing=5, interpSpeedIncreasing=15}
    boundBone {boneName=dyn_shorts_l_04}
    componentLinearAccScale=<1 1 1>
    componentLinearVelScale=<1 1 1>
    gravityOverride {z=-980}
    physicsBodyDefinitions [{
      boundBone {boneName=dyn_shorts_l_04}
      boxExtents=<20 30 30>
      constraintSetup {angularLimitsMax=<10 10 10>, angularLimitsMin=<-10 -10 -10>, angularTarget=<1 0 0>}
      localJointOffset=<1 0 0>
     }]
   }
   n19 = Anim Dynamics - Bone: dyn_shorts_l_02(ComponentPose=@n18, Alpha=0.3, ExternalForce="0,0,0") {
    angularDampingOverride=1
    angularSpringConstant=20000
    bAngularSpring=true
    bOverrideAngularDamping=true
    bUseGravityOverride=true
    gravityScale=0.35
    simulationSpace=RootRelative
    alphaScaleBiasClamp {bInterpResult=true, interpSpeedDecreasing=5, interpSpeedIncreasing=15}
    boundBone {boneName=dyn_shorts_l_02}
    componentLinearAccScale=<1 1 1>
    componentLinearVelScale=<1 1 1>
    gravityOverride {z=-980}
    physicsBodyDefinitions [{
      boundBone {boneName=dyn_shorts_l_02}
      boxExtents=<20 30 30>
      constraintSetup {angularLimitsMax=<10 10 10>, angularLimitsMin=<-10 -10 -10>, angularTarget=<1 0 0>}
      localJointOffset=<1 0 0>
     }]
   }
   n20 = Anim Dynamics - Bone: dyn_shorts_r_02(ComponentPose=@n19, Alpha=0.3, ExternalForce="0,0,0") {
    angularDampingOverride=1
    angularSpringConstant=20000
    bAngularSpring=true
    bOverrideAngularDamping=true
    bUseGravityOverride=true
    gravityScale=0.35
    simulationSpace=RootRelative
    alphaScaleBiasClamp {bInterpResult=true, interpSpeedDecreasing=5, interpSpeedIncreasing=15}
    boundBone {boneName=dyn_shorts_r_02}
    componentLinearAccScale=<1 1 1>
    componentLinearVelScale=<1 1 1>
    gravityOverride {z=-980}
    physicsBodyDefinitions [{
      boundBone {boneName=dyn_shorts_r_02}
      boxExtents=<20 30 30>
      constraintSetup {angularLimitsMax=<10 10 10>, angularLimitsMin=<-10 -10 -10>, angularTarget=<1 0 0>}
      localJointOffset=<2 0 0>
     }]
    sphericalLimits [{limitRadius=18, drivingBone {boneName=deform_knee_in_r}, sphereLocalOffset=<0 5 10>}]
   }
   n21 = Anim Dynamics - Bone: dyn_shorts_r_03(ComponentPose=@n20, Alpha=0.3, ExternalForce="0,0,0") {
    angularDampingOverride=1
    angularSpringConstant=20000
    bAngularSpring=true
    bOverrideAngularDamping=true
    bUseGravityOverride=true
    bUsePlanarLimit=false
    bUseSphericalLimits=true
    gravityScale=0.35
    simulationSpace=RootRelative
    alphaScaleBiasClamp {bInterpResult=true, interpSpeedDecreasing=5, interpSpeedIncreasing=15}
    boundBone {boneName=dyn_shorts_r_03}
    componentLinearAccScale=<1 1 1>
    componentLinearVelScale=<1 1 1>
    gravityOverride {z=-980}
    physicsBodyDefinitions [{
      boundBone {boneName=dyn_shorts_r_03}
      boxExtents=<20 30 30>
      constraintSetup {angularLimitsMax=<10 0 10>, angularLimitsMin=<-10 -10 -10>, angularTarget=<1 0 0>}
      localJointOffset=<1 0 0>
     }]
    sphericalLimits [{limitRadius=10, drivingBone {boneName=deform_knee_in_r}, sphereLocalOffset=<0 0 -1>}]
   }
   n22 = Anim Dynamics - Bone: dyn_shorts_r_04(ComponentPose=@n21, Alpha=1, ExternalForce="0,0,0") {
    angularDampingOverride=1
    angularSpringConstant=20000
    bAngularSpring=true
    bOverrideAngularDamping=true
    bUseGravityOverride=true
    gravityScale=0.35
    simulationSpace=RootRelative
    alphaScaleBiasClamp {bInterpResult=true, interpSpeedDecreasing=5, interpSpeedIncreasing=15}
    boundBone {boneName=dyn_shorts_r_04}
    componentLinearAccScale=<1 1 1>
    componentLinearVelScale=<1 1 1>
    gravityOverride {z=-980}
    physicsBodyDefinitions [{
      boundBone {boneName=dyn_shorts_r_04}
      boxExtents=<20 30 30>
      constraintSetup {angularLimitsMax=<10 10 10>, angularLimitsMin=<-10 -10 -10>, angularTarget=<1 0 0>}
      localJointOffset=<2 0 0>
     }]
    sphericalLimits [{limitRadius=15, drivingBone {boneName=deform_knee_in_r}, sphereLocalOffset=<-10 10 -10>}]
   }
   n23 = Anim Dynamics - Bone: dyn_shorts_r_01(ComponentPose=@n22, Alpha=1, ExternalForce="0,0,0") {
    angularDampingOverride=1
    angularSpringConstant=20000
    bAngularSpring=true
    bOverrideAngularDamping=true
    bUseGravityOverride=true
    gravityScale=0.35
    simulationSpace=RootRelative
    alphaScaleBiasClamp {bInterpResult=true, interpSpeedDecreasing=5, interpSpeedIncreasing=15}
    boundBone {boneName=dyn_shorts_r_01}
    componentLinearAccScale=<1 1 1>
    componentLinearVelScale=<1 1 1>
    gravityOverride {z=-980}
    physicsBodyDefinitions [{
      boundBone {boneName=dyn_shorts_r_01}
      boxExtents=<20 30 30>
      constraintSetup {angularLimitsMax=<0 0 5>, angularLimitsMin=<0 0 -5>, angularTarget=<1 0 0>}
      localJointOffset=<1 0 0>
     }]
    sphericalLimits [{limitRadius=12, drivingBone {boneName=deform_knee_in_r}, sphereLocalOffset=<0 0 0>}]
   }
   n24 = Anim Dynamics - Bone: dyn_string(ComponentPose=@n23, Alpha=1, ExternalForce="0,0,0") {
    angularDampingOverride=1
    angularSpringConstant=20000
    bAngularSpring=true
    bOverrideAngularDamping=true
    bUseGravityOverride=true
    gravityScale=0.35
    simulationSpace=RootRelative
    boundBone {boneName=dyn_string}
    componentLinearAccScale=<1 1 1>
    componentLinearVelScale=<1 1 1>
    gravityOverride {z=-980}
    physicsBodyDefinitions [{
      boundBone {boneName=dyn_string}
      boxExtents=<20 30 30>
      constraintSetup {angularLimitsMax=<0 5 0>, angularLimitsMin=<0 -5 0>, angularTarget=<1 0 0>}
      localJointOffset=<1 0 0>
     }]
   }
   n25 = Anim Dynamics - Bone: dyn_string_04(ComponentPose=@n24, Alpha=1, ExternalForce="0,0,0") {
    angularDampingOverride=1
    angularSpringConstant=20000
    bAngularSpring=true
    bOverrideAngularDamping=true
    bUseGravityOverride=true
    gravityScale=0.35
    simulationSpace=RootRelative
    boundBone {boneName=dyn_string_04}
    componentLinearAccScale=<1 1 1>
    componentLinearVelScale=<1 1 1>
    gravityOverride {z=-980}
    physicsBodyDefinitions [{
      boundBone {boneName=dyn_string_04}
      boxExtents=<20 30 30>
      constraintSetup {angularLimitsMax=<0 5 15>, angularLimitsMin=<0 -5 -5>, angularTarget=<1 0 0>}
      localJointOffset=<1 0 0>
     }]
    planarLimits [{drivingBone {boneName=thigh_l}, planeTransform {location=<0 0 0>, rotation=<rot 0 0 80.0001>}}]
   }
   n26 = Anim Dynamics - Bone: dyn_string_03(ComponentPose=@n25, Alpha=1, ExternalForce="0,0,0") {
    angularDampingOverride=1
    angularSpringConstant=20000
    bAngularSpring=true
    bOverrideAngularDamping=true
    bUseGravityOverride=true
    bUsePlanarLimit=false
    gravityScale=0.35
    simulationSpace=RootRelative
    boundBone {boneName=dyn_string_03}
    componentLinearAccScale=<1 1 1>
    componentLinearVelScale=<1 1 1>
    gravityOverride {z=-980}
    physicsBodyDefinitions [{
      boundBone {boneName=dyn_string_03}
      boxExtents=<20 30 30>
      constraintSetup {angularLimitsMax=<0 5 15>, angularLimitsMin=<0 -10 -5>, angularTarget=<1 0 0>}
      localJointOffset=<4 0 0>
     }]
    planarLimits [{drivingBone {boneName=base_string}, planeTransform {location=<0 0 0>, rotation=<rot 0 0 80.0001>}}]
   }
   n27 = Anim Dynamics - Bone: dyn_string_02(ComponentPose=@n26, Alpha=1, ExternalForce="0,0,0") {
    angularDampingOverride=1
    angularSpringConstant=20000
    bAngularSpring=true
    bOverrideAngularDamping=true
    bUseGravityOverride=true
    bUsePlanarLimit=false
    gravityScale=0.35
    simulationSpace=RootRelative
    boundBone {boneName=dyn_string_02}
    componentLinearAccScale=<1 1 1>
    componentLinearVelScale=<1 1 1>
    gravityOverride {z=-980}
    physicsBodyDefinitions [{
      boundBone {boneName=dyn_string_02}
      boxExtents=<20 30 30>
      constraintSetup {angularLimitsMax=<0 5 0>, angularLimitsMin=<0 -10 -15>, angularTarget=<1 0 0>}
      localJointOffset=<4 0 0>
     }]
    planarLimits [{drivingBone {boneName=base_string}, planeTransform {location=<0 1 0>, rotation=<rot 0 0 80.0001>}}]
   }
   n28 = Anim Dynamics - Bone: dyn_string_01(ComponentPose=@n27, Alpha=1, ExternalForce="0,0,0") {
    angularDampingOverride=1
    angularSpringConstant=20000
    bAngularSpring=true
    bOverrideAngularDamping=true
    bUseGravityOverride=true
    bUsePlanarLimit=false
    gravityScale=0.35
    simulationSpace=RootRelative
    boundBone {boneName=dyn_string_01}
    componentLinearAccScale=<1 1 1>
    componentLinearVelScale=<1 1 1>
    gravityOverride {z=-980}
    physicsBodyDefinitions [{
      boundBone {boneName=dyn_string_01}
      boxExtents=<20 30 30>
      constraintSetup {angularLimitsMax=<0 5 0>, angularLimitsMin=<0 -5 -15>, angularTarget=<1 0 0>}
      localJointOffset=<4 0 0>
     }]
    planarLimits [{drivingBone {boneName=thigh_r}, planeTransform {location=<0 -8 0>, rotation=<rot 0 -20 -120.0001>}}]
   }
   n29 = "Transform (Modify) Bone - Bone: base_string"!(ComponentPose=@n28, bAlphaBoolEnabled:bool=True, Translation="0,0,0", Rotation="10,5,-50", Scale="1,1,1") -> (Pose) {
    alphaInputType=Bool
    rotationMode=BMM_Additive
    rotationSpace=BCS_BoneSpace
    translationMode=BMM_Additive
    translationSpace=BCS_BoneSpace
    alphaBoolBlend {blendInTime=0.2, blendOutTime=0.1}
    boneToModify {boneName=base_string}
    rotation=<rot 10 5 -50>
   }
   n30 = "Transform (Modify) Bone - Bone: base_string_04"!(ComponentPose=@n29, bAlphaBoolEnabled:bool=True, Translation="0,0,0", Rotation="-60,0,-10", Scale="1,1,1") -> (Pose) {
    alphaInputType=Bool
    rotationMode=BMM_Additive
    rotationSpace=BCS_BoneSpace
    translationMode=BMM_Additive
    translationSpace=BCS_BoneSpace
    alphaBoolBlend {blendInTime=0.2, blendOutTime=0.1}
    boneToModify {boneName=base_string_04}
    rotation {pitch=-60, roll=-10}
   }
   n31 = "Transform (Modify) Bone - Bone: deform_bicep_r"!(
    ComponentPose=@n30
    bAlphaBoolEnabled:bool=Get bIsRightArmHidden() -> (bIsRightArmHidden:bool)
    Translation="-20,0,15"
    Rotation="0,0,90"
    Scale="0.5,0.5,0.5"
   ) -> (Pose) {
    alphaInputType=Bool
    rotationMode=BMM_Additive
    rotationSpace=BCS_BoneSpace
    scaleMode=BMM_Additive
    scaleSpace=BCS_BoneSpace
    translationMode=BMM_Additive
    translationSpace=BCS_BoneSpace
    alphaBoolBlend {blendInTime=0.2, blendOutTime=0.1}
    boneToModify {boneName=deform_bicep_r}
    rotation {roll=90}
    scale=<0.5 0.5 0.5>
    translation {x=-20, z=15}
   }
   n32 = "Transform (Modify) Bone - Bone: deform_elbow_in_l"!(
    ComponentPose=@n31
    bAlphaBoolEnabled:bool=Get bIsLeftArmHidden() -> (bIsLeftArmHidden:bool)
    Translation="4,0,-11"
    Rotation="0,0,90"
    Scale="0.5,0.5,1"
   ) -> (Pose) {
    alphaInputType=Bool
    rotationMode=BMM_Additive
    rotationSpace=BCS_BoneSpace
    scaleMode=BMM_Additive
    scaleSpace=BCS_BoneSpace
    translationMode=BMM_Additive
    translationSpace=BCS_BoneSpace
    alphaBoolBlend {blendInTime=0.2, blendOutTime=0.1}
    boneToModify {boneName=deform_elbow_in_l}
    rotation {roll=90}
    scale=<0.5 0.5>
    translation {x=4, z=-11}
   }
   n33 = Control Rig(Source=Use cached pose 'Input_pose'()) {controlRigAssetReference {controlRigAsset=&M_LRG_OtterTeach_CtrlRig}}
   n34 = Anim Dynamics - Bone: dyn_shorts_l_01(ComponentPose=Local To Component(LocalPose=@n33), Alpha=1, ExternalForce="0,0,0") {
    angularDampingOverride=1
    angularSpringConstant=20000
    bAngularSpring=true
    bOverrideAngularDamping=true
    bUseGravityOverride=true
    gravityScale=0.35
    simulationSpace=RootRelative
    alphaScaleBiasClamp {bInterpResult=true, interpSpeedDecreasing=5, interpSpeedIncreasing=15}
    boundBone {boneName=dyn_shorts_l_01}
    componentLinearAccScale=<1 1 1>
    componentLinearVelScale=<1 1 1>
    gravityOverride {z=-980}
    physicsBodyDefinitions [{
      boundBone {boneName=dyn_shorts_l_01}
      boxExtents=<20 30 30>
      constraintSetup {angularLimitsMax=<0 0 5>, angularLimitsMin=<0 0 -5>, angularTarget=<1 0 0>}
      localJointOffset=<1 0 0>
     }]
    simSpaceSettings {simSpaceAngularAlpha=1}
   }
   n35 = Anim Dynamics - Bone: dyn_shorts_l_04(ComponentPose=@n34, Alpha=1, ExternalForce="0,0,0") {
    angularDampingOverride=1
    angularSpringConstant=20000
    bAngularSpring=true
    bOverrideAngularDamping=true
    bUseGravityOverride=true
    gravityScale=0.35
    simulationSpace=RootRelative
    alphaScaleBiasClamp {bInterpResult=true, interpSpeedDecreasing=5, interpSpeedIncreasing=15}
    boundBone {boneName=dyn_shorts_l_04}
    componentLinearAccScale=<1 1 1>
    componentLinearVelScale=<1 1 1>
    gravityOverride {z=-980}
    physicsBodyDefinitions [{
      boundBone {boneName=dyn_shorts_l_04}
      boxExtents=<20 30 30>
      constraintSetup {angularLimitsMax=<10 10 10>, angularLimitsMin=<-10 -10 -10>, angularTarget=<1 0 0>}
      localJointOffset=<1 0 0>
     }]
    simSpaceSettings {simSpaceAngularAlpha=1}
   }
   n36 = Anim Dynamics - Bone: dyn_shorts_l_02(ComponentPose=@n35, Alpha=0.3, ExternalForce="0,0,0") {
    angularDampingOverride=1
    angularSpringConstant=20000
    bAngularSpring=true
    bOverrideAngularDamping=true
    bUseGravityOverride=true
    gravityScale=0.35
    simulationSpace=RootRelative
    alphaScaleBiasClamp {bInterpResult=true, interpSpeedDecreasing=5, interpSpeedIncreasing=15}
    boundBone {boneName=dyn_shorts_l_02}
    componentLinearAccScale=<1 1 1>
    componentLinearVelScale=<1 1 1>
    gravityOverride {z=-980}
    physicsBodyDefinitions [{
      boundBone {boneName=dyn_shorts_l_02}
      boxExtents=<20 30 30>
      constraintSetup {angularLimitsMax=<10 10 10>, angularLimitsMin=<-10 -10 -10>, angularTarget=<1 0 0>}
      localJointOffset=<1 0 0>
     }]
    simSpaceSettings {simSpaceAngularAlpha=1}
   }
   n37 = Anim Dynamics - Bone: dyn_shorts_r_02(ComponentPose=@n36, Alpha=0.3, ExternalForce="0,0,0") {
    angularDampingOverride=1
    angularSpringConstant=20000
    bAngularSpring=true
    bOverrideAngularDamping=true
    bUseGravityOverride=true
    gravityScale=0.35
    simulationSpace=RootRelative
    alphaScaleBiasClamp {bInterpResult=true, interpSpeedDecreasing=5, interpSpeedIncreasing=15}
    boundBone {boneName=dyn_shorts_r_02}
    componentLinearAccScale=<1 1 1>
    componentLinearVelScale=<1 1 1>
    gravityOverride {z=-980}
    physicsBodyDefinitions [{
      boundBone {boneName=dyn_shorts_r_02}
      boxExtents=<20 30 30>
      constraintSetup {angularLimitsMax=<10 10 10>, angularLimitsMin=<-10 -10 -10>, angularTarget=<1 0 0>}
      localJointOffset=<1 0 0>
     }]
    simSpaceSettings {simSpaceAngularAlpha=1}
    sphericalLimits [{limitRadius=18, drivingBone {boneName=deform_knee_in_r}, sphereLocalOffset=<0 5 10>}]
   }
   n38 = Anim Dynamics - Bone: dyn_shorts_r_03(ComponentPose=@n37, Alpha=0.3, ExternalForce="0,0,0") {
    angularDampingOverride=1
    angularSpringConstant=20000
    bAngularSpring=true
    bOverrideAngularDamping=true
    bUseGravityOverride=true
    bUsePlanarLimit=false
    gravityScale=0.35
    simulationSpace=RootRelative
    alphaScaleBiasClamp {bInterpResult=true, interpSpeedDecreasing=5, interpSpeedIncreasing=15}
    boundBone {boneName=dyn_shorts_r_03}
    componentLinearAccScale=<1 1 1>
    componentLinearVelScale=<1 1 1>
    gravityOverride {z=-980}
    physicsBodyDefinitions [{
      boundBone {boneName=dyn_shorts_r_03}
      boxExtents=<20 30 30>
      constraintSetup {angularLimitsMax=<5 0 5>, angularLimitsMin=<-5 0 -5>, angularTarget=<1 0 0>}
      localJointOffset=<1 0 0>
     }]
    simSpaceSettings {simSpaceAngularAlpha=1}
    sphericalLimits [{limitRadius=10, drivingBone {boneName=deform_knee_in_r}, sphereLocalOffset=<0 0 0>}]
   }
   n39 = Anim Dynamics - Bone: dyn_shorts_r_04(ComponentPose=@n38, Alpha=1, ExternalForce="0,0,0") {
    angularDampingOverride=1
    angularSpringConstant=20000
    bAngularSpring=true
    bOverrideAngularDamping=true
    bUseGravityOverride=true
    gravityScale=0.35
    simulationSpace=RootRelative
    alphaScaleBiasClamp {bInterpResult=true, interpSpeedDecreasing=5, interpSpeedIncreasing=15}
    boundBone {boneName=dyn_shorts_r_04}
    componentLinearAccScale=<1 1 1>
    componentLinearVelScale=<1 1 1>
    gravityOverride {z=-980}
    physicsBodyDefinitions [{
      boundBone {boneName=dyn_shorts_r_04}
      boxExtents=<20 30 30>
      constraintSetup {angularLimitsMax=<10 10 10>, angularLimitsMin=<-10 -10 -10>, angularTarget=<1 0 0>}
      localJointOffset=<1 0 0>
     }]
    simSpaceSettings {simSpaceAngularAlpha=1}
    sphericalLimits [{limitRadius=15, drivingBone {boneName=deform_knee_in_r}, sphereLocalOffset=<-10 10 -10>}]
   }
   n40 = Anim Dynamics - Bone: dyn_shorts_r_01(ComponentPose=@n39, Alpha=1, ExternalForce="0,0,0") {
    angularDampingOverride=1
    angularSpringConstant=20000
    bAngularSpring=true
    bOverrideAngularDamping=true
    bUseGravityOverride=true
    gravityScale=0.35
    simulationSpace=RootRelative
    alphaScaleBiasClamp {bInterpResult=true, interpSpeedDecreasing=5, interpSpeedIncreasing=15}
    boundBone {boneName=dyn_shorts_r_01}
    componentLinearAccScale=<1 1 1>
    componentLinearVelScale=<1 1 1>
    gravityOverride {z=-980}
    physicsBodyDefinitions [{
      boundBone {boneName=dyn_shorts_r_01}
      boxExtents=<20 30 30>
      constraintSetup {angularLimitsMax=<0 0 5>, angularLimitsMin=<0 0 -5>, angularTarget=<1 0 0>}
      localJointOffset=<1 0 0>
     }]
    simSpaceSettings {simSpaceAngularAlpha=1}
    sphericalLimits [{limitRadius=12, drivingBone {boneName=deform_knee_in_r}, sphereLocalOffset=<0 0 0>}]
   }
   n41 = Anim Dynamics - Bone: dyn_string(ComponentPose=@n40, Alpha=1, ExternalForce="0,0,0") {
    angularDampingOverride=1
    angularSpringConstant=20000
    bAngularSpring=true
    bOverrideAngularDamping=true
    bUseGravityOverride=true
    gravityScale=0.35
    simulationSpace=RootRelative
    boundBone {boneName=dyn_string}
    componentLinearAccScale=<1 1 1>
    componentLinearVelScale=<1 1 1>
    gravityOverride {z=-980}
    physicsBodyDefinitions [{
      boundBone {boneName=dyn_string}
      boxExtents=<20 30 30>
      constraintSetup {angularLimitsMax=<0 5 0>, angularLimitsMin=<0 -5 0>, angularTarget=<1 0 0>}
      localJointOffset=<1 0 0>
     }]
   }
   n42 = Anim Dynamics - Bone: dyn_string_04(ComponentPose=@n41, Alpha=1, ExternalForce="0,0,0") {
    angularDampingOverride=1
    angularSpringConstant=20000
    bAngularSpring=true
    bOverrideAngularDamping=true
    bUseGravityOverride=true
    gravityScale=0.35
    simulationSpace=RootRelative
    boundBone {boneName=dyn_string_04}
    componentLinearAccScale=<1 1 1>
    componentLinearVelScale=<1 1 1>
    gravityOverride {z=-980}
    physicsBodyDefinitions [{
      boundBone {boneName=dyn_string_04}
      boxExtents=<20 30 30>
      constraintSetup {angularLimitsMax=<0 5 15>, angularLimitsMin=<0 -5 -5>, angularTarget=<1 0 0>}
      localJointOffset=<1 0 0>
     }]
    planarLimits [{drivingBone {boneName=thigh_l}, planeTransform {location=<0 0 0>, rotation=<rot 0 0 80.0001>}}]
   }
   n43 = Anim Dynamics - Bone: dyn_string_03(ComponentPose=@n42, Alpha=1, ExternalForce="0,0,0") {
    angularDampingOverride=1
    angularSpringConstant=20000
    bAngularSpring=true
    bOverrideAngularDamping=true
    bUseGravityOverride=true
    bUsePlanarLimit=false
    gravityScale=0.35
    simulationSpace=RootRelative
    boundBone {boneName=dyn_string_03}
    componentLinearAccScale=<1 1 1>
    componentLinearVelScale=<1 1 1>
    gravityOverride {z=-980}
    physicsBodyDefinitions [{
      boundBone {boneName=dyn_string_03}
      boxExtents=<20 30 30>
      constraintSetup {angularLimitsMax=<0 5 15>, angularLimitsMin=<0 -10 -5>, angularTarget=<1 0 0>}
      localJointOffset=<4 0 0>
     }]
    planarLimits [{drivingBone {boneName=base_string}, planeTransform {location=<0 0 0>, rotation=<rot 0 0 80.0001>}}]
   }
   n44 = Anim Dynamics - Bone: dyn_string_02(ComponentPose=@n43, Alpha=1, ExternalForce="0,0,0") {
    angularDampingOverride=1
    angularSpringConstant=20000
    bAngularSpring=true
    bOverrideAngularDamping=true
    bUseGravityOverride=true
    bUsePlanarLimit=false
    gravityScale=0.35
    simulationSpace=RootRelative
    boundBone {boneName=dyn_string_02}
    componentLinearAccScale=<1 1 1>
    componentLinearVelScale=<1 1 1>
    gravityOverride {z=-980}
    physicsBodyDefinitions [{
      boundBone {boneName=dyn_string_02}
      boxExtents=<20 30 30>
      constraintSetup {angularLimitsMax=<0 5 0>, angularLimitsMin=<0 -10 -15>, angularTarget=<1 0 0>}
      localJointOffset=<4 0 0>
     }]
    planarLimits [{drivingBone {boneName=base_string}, planeTransform {location=<0 1 0>, rotation=<rot 0 0 80.0001>}}]
   }
   n45 = Anim Dynamics - Bone: dyn_string_01(ComponentPose=@n44, Alpha=1, ExternalForce="0,0,0") {
    angularDampingOverride=1
    angularSpringConstant=20000
    bAngularSpring=true
    bOverrideAngularDamping=true
    bUseGravityOverride=true
    bUsePlanarLimit=false
    gravityScale=0.35
    simulationSpace=RootRelative
    boundBone {boneName=dyn_string_01}
    componentLinearAccScale=<1 1 1>
    componentLinearVelScale=<1 1 1>
    gravityOverride {z=-980}
    physicsBodyDefinitions [{
      boundBone {boneName=dyn_string_01}
      boxExtents=<20 30 30>
      constraintSetup {angularLimitsMax=<0 5 0>, angularLimitsMin=<0 -5 -15>, angularTarget=<1 0 0>}
      localJointOffset=<4 0 0>
     }]
    planarLimits [{drivingBone {boneName=thigh_r}, planeTransform {location=<0 -8 0>, rotation=<rot 0 -20 -120.0001>}}]
   }
   n46 = "Transform (Modify) Bone - Bone: base_string"(ComponentPose=@n45, bAlphaBoolEnabled=True, Translation="0,0,0", Rotation="10,5,-50", Scale="1,1,1") {
    alphaInputType=Bool
    rotationMode=BMM_Additive
    rotationSpace=BCS_BoneSpace
    translationMode=BMM_Additive
    translationSpace=BCS_BoneSpace
    alphaBoolBlend {blendInTime=0.2, blendOutTime=0.1}
    boneToModify {boneName=base_string}
    rotation=<rot 10 5 -50>
   }
   n47 = "Transform (Modify) Bone - Bone: base_string_04"(ComponentPose=@n46, bAlphaBoolEnabled=True, Translation="0,0,0", Rotation="-60,0,-10", Scale="1,1,1") {
    alphaInputType=Bool
    rotationMode=BMM_Additive
    rotationSpace=BCS_BoneSpace
    translationMode=BMM_Additive
    translationSpace=BCS_BoneSpace
    alphaBoolBlend {blendInTime=0.2, blendOutTime=0.1}
    boneToModify {boneName=base_string_04}
    rotation {pitch=-60, roll=-10}
   }
   Output Pose(Result=Blend Poses by bool(BlendPose_0=Component To Local(ComponentPose=@n47), BlendPose_1=Component To Local(ComponentPose=@n32))) {layerGroup=None}
   Save cached pose 'Input_pose'(Pose=Copy Pose From Mesh() {bUseAttachedParent=true}) {cacheName=Input_pose}
  ]}
 EventGraph {nodes [
   Event BlueprintUpdateAnimation()
   > n1 = Gravity Override State Function DEPRECATED(CharacterPartAnimInstance=Self-Reference(), GameTimeInSeconds=GetGameTimeInSeconds(), GravityOverrideParams=Get Gravity Override Params())
   > Set Gravity Override(Gravity Override=@n1)
   > n2 = Branch(Condition=OR Boolean!(A=Get bIsPlayingEmote(), B=Get bIsKneeSliding(), C:bool=Get bIsCrouching(), D:bool=Get bIsOnGrindRail()) -> (ReturnValue))
   > Set on_AnimDynamics(on_AnimDynamics:double=0.3) -> (Output_Get:double)
   Set on_AnimDynamics(execute=@n2.else, on_AnimDynamics=0.9)
   n3 = Get bIsMoving2D()
   Set Only_Crouching(Only_Crouching:bool=AND Boolean(A=NOT Boolean(A=@n3), B=Get bIsCrouching())) -> (Output_Get:bool)
   n4 = Get bIsMoving2D()
   > Set Only_Jumping(Only_Jumping:bool=AND Boolean(A=NOT Boolean(A=@n4), B=Get bIsJumping())) -> (Output_Get:bool)
  ]}
]
variables [
 Gravity Override {type=Vector}
 Gravity Override Params {type=GravityOverrideParamsStruct}
 on_AnimDynamics {type=number}
 Only_Crouching {type=boolean}
 Only_Jumping {type=boolean}
]
