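# Physics asset for the left leg mesh of the large male OtterTeach player body cosmetic, apparently for secondary motion of the shorts.
# Four capsule bodies on the dyn_shorts_l bones, the last one shorter, sit alongside a larger capsule on thigh_l.
# Each shorts bone is constrained directly to the thigh rather than chained, with Swing2 locked and small per-bone angle settings.
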PhysicsAsset /BRCosmetics/Characters/Player/Male/Large/Bodies/M_LRG_OtterTeach/Meshes/M_LRG_OtterTeach_Left_leg_Physics
bodies {
 dyn_shorts_l_01 [{shapeType=Capsule, length=12, radius=3, shapeName=dyn_shorts_l_01_capsule, center=<8.3256 0 0>, rotation=<rot 90 0 0>}]
 dyn_shorts_l_02 [{shapeType=Capsule, length=12, radius=3, shapeName=dyn_shorts_l_01_capsule, center=<8.3256 0 0>, rotation=<rot 90 0 0>}]
 dyn_shorts_l_03 [{shapeType=Capsule, length=12, radius=3, shapeName=dyn_shorts_l_01_capsule, center=<8.3256 0 0>, rotation=<rot 90 0 0>}]
 dyn_shorts_l_04 [{shapeType=Capsule, length=6, radius=3, shapeName=dyn_shorts_l_01_capsule, center=<5.129 0 0>, rotation=<rot 90 0 0>}]
 thigh_l [{shapeType=Capsule, length=20, radius=8, shapeName=thigh_r_capsule, center=<-23.6511 -0.3181 1.4549>, rotation=<rot 87.6723 179.9999 180>}]
}
constraints [
 {bone1Name=dyn_shorts_l_01, bone2Name=thigh_l, swing1LimitDegrees=10, swing2LimitDegrees=15, swing2Motion=Locked, twistLimitDegrees=1}
 {bone1Name=dyn_shorts_l_02, bone2Name=thigh_l, swing1LimitDegrees=10, swing2LimitDegrees=10, swing2Motion=Locked, twistLimitDegrees=1}
 {bone1Name=dyn_shorts_l_03, bone2Name=thigh_l, swing1LimitDegrees=2, swing2LimitDegrees=10, swing2Motion=Locked, twistLimitDegrees=1}
 {bone1Name=dyn_shorts_l_04, bone2Name=thigh_l, swing1LimitDegrees=10, swing2LimitDegrees=7, swing2Motion=Locked, twistLimitDegrees=1}
]
